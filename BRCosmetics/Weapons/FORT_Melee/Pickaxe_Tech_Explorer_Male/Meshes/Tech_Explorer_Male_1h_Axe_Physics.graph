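# Physics asset for the Tech Explorer Male one-handed axe, a pickaxe cosmetic mesh, setting up physics for its dyn_loop bone.
# It has two capsule bodies: a small one on the Melee_TwoHanded_Handle bone and a larger, offset and rotated one on dyn_loop.
# A single constraint joins the loop to the handle, with wide swing limits and a narrow twist limit.
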PhysicsAsset /BRCosmetics/Weapons/FORT_Melee/Pickaxe_Tech_Explorer_Male/Meshes/Tech_Explorer_Male_1h_Axe_Physics
bodies {
 Melee_TwoHanded_Handle [{shapeType=Capsule, length=1, radius=1}]
 dyn_loop [{shapeType=Capsule, length=5, radius=3.75, center=<5 0 0>, rotation=<rot 90 0 0>}]
}
constraints [{bone1Name=dyn_loop, bone2Name=Melee_TwoHanded_Handle, swing1LimitDegrees=65, swing2LimitDegrees=80, twistLimitDegrees=11}]
